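# Animation Blueprint for the M_MED_StageCue_Chalk player body cosmetic, which follows the pose of the mesh it is attached to.
# Two branches each copy the attached parent's pose and run it through a Control Rig, one front-end variant and one standard.
# A bool blend on bIsInFrontEnd selects the front-end rig when true and the standard rig otherwise.
# The update event holds no logic.
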AnimBlueprint /BRCosmetics/Characters/Player/Male/Medium/Bodies/M_MED_StageCue_Chalk/Meshes/M_MED_StageCue_Chalk_AnimBP
target_skeleton=&/M_MED_StageCue_Chalk_Skeleton
graphs [
 AnimGraph {nodes [
   n1 = Control Rig(Source=Copy Pose From Mesh() {bUseAttachedParent=true}) {
    alphaBoolBlend {blendOption=Linear}
    controlRigAssetReference {blueprintRigClass=&/M_MED_StageCue_Chalk_Frontend_CtrlRig.M_MED_StageCue_Chalk_Frontend_CtrlRig_C}
   }
   n2 = Control Rig(Source=Copy Pose From Mesh() {bUseAttachedParent=true}) {
    alphaBoolBlend {blendOption=Linear}
    controlRigAssetReference {blueprintRigClass=&/M_MED_StageCue_Chalk_CtrlRig.M_MED_StageCue_Chalk_CtrlRig_C}
   }
   Output Pose(Result=Blend Poses by bool(BlendPose_0=@n1, BlendPose_1=@n2, bActiveValue=Get bIsInFrontEnd()) {blendType=Linear}) {layerGroup=None}
  ]}
 EventGraph {nodes [Event BlueprintUpdateAnimation()]}
]
variables []
