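# Animation Blueprint for the face accessory part of the StageCue Chalk medium male character.
# Copies the pose from the attached parent mesh, then runs it through the accessory's Control Rig with a Valid Character flag.
# On customization load, it caches the head mesh and sets that flag by matching a custom part's mesh name to the Chalk jacket.
# The update event is empty, and the negated head-to-neck_02 rotation math is unconnected.
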AnimBlueprint /BRCosmetics/Characters/Player/Male/Medium/Bodies/M_MED_StageCue_Chalk/Meshes/Parts/M_MED_StageCue_Chalk_FaceAcc_AnimBP
target_skeleton=&/M_MED_StageCue_Chalk_FaceAcc_Skeleton
graphs [
 AnimGraph {nodes [
   n1 = Control Rig!(
    Source=Copy Pose From Mesh() {bUseAttachedParent=true}
    Alpha
    Valid Character:bool=Get M_MED_StageCue_ChalkJacket() -> (M_MED_StageCue_ChalkJacket:bool)
   ) -> (Pose) {
    alphaBoolBlend {blendOption=Linear}
    controlRigAssetReference {blueprintRigClass=&/M_MED_StageCue_Chalk_FaceAcc_ControlRig.M_MED_StageCue_Chalk_FaceAcc_ControlRig_C}
   }
   Output Pose(Result=@n1) {layerGroup=None}
  ]}
 EventGraph {nodes [
   n1 = Event OnCharacterCustomizationLoaded()
   > Set Head Mesh(Head Mesh:Skeletal Mesh Component Budgeted Object Reference=Get Array Item with EFort Custom Part Type(Array=@n1.CustomMeshParts, PartType=Head)) -> (Output_Get:Skeletal Mesh Component Budgeted Object Reference)
   n2 = GetSkinnedAsset!(self:Skinned Mesh Component Object Reference=Get Array Item with EFort Custom Part Type(Array=@n1.CustomMeshParts)) -> (ReturnValue:Skinned Asset Object Reference)
   > Set M_MED_StageCue_ChalkJacket(M_MED_StageCue_ChalkJacket:bool="Equal Exactly (String)"(A=GetDisplayName(Object=@n2), B=M_MED_StageCue_ChalkJacket)) -> (Output_Get:bool)
   n3 = BreakRotator(InRot=GetDeltaTransformFromRefPose!(self=Get Head Mesh() -> (Head Mesh:Skeletal Mesh Component Budgeted Object Reference), BoneName=head, BaseName=neck_02) -> (ReturnValue_Location:Vector, ReturnValue_Rotation:Rotator, ReturnValue_Scale:Vector).ReturnValue_Rotation)
   MakeRotator(
    Roll=float * float!(A=@n3, B=-1, C:double) -> (ReturnValue) {is=title K2Node_PromotableOperator}
    Pitch=float * float(A=@n3.Pitch, B=-1)
    Yaw=float * float(A=@n3.Yaw, B=-1)
   )
   Event BlueprintUpdateAnimation()
  ]}
]
variables [
 Head Mesh {type=SkeletalMeshComponentBudgeted}
 M_MED_StageCue_ChalkJacket {type=boolean}
]
